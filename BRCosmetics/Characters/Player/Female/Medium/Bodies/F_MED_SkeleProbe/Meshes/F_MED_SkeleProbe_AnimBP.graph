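# Animation Blueprint for the F_MED_SkeleProbe player body, copying the attached parent mesh's pose and correcting collar, skirt and pelvis bones.
# Weighted constraints drive collar and back skirt bones, then additive offsets adjust the pelvis when crouching and skirt when parachuting.
# A Control Rig finishes the pose using pistol and not-emoting alphas.
# The update event derives alphas from player state and weapon animation type; the rifle alpha is unused here.
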AnimBlueprint /BRCosmetics/Characters/Player/Female/Medium/Bodies/F_MED_SkeleProbe/Meshes/F_MED_SkeleProbe_AnimBP
target_skeleton=&/F_MED_SkeleProbe_Skeleton
graphs [
 AnimGraph {nodes [
   n1 = "Constraint - Collar_bk_r_MSTR (RRR)"!(
    ComponentPose=Local To Component(LocalPose=Copy Pose From Mesh() {bUseAttachedParent=true})
    Alpha=1
    ConstraintWeights_0=0.25
    ConstraintWeights_1:float=0.25
    ConstraintWeights_2:float=0.5
   ) -> (Pose) {
    alphaBoolBlend {blendOption=Linear}
    boneToModify {boneName=Collar_bk_r_MSTR}
    constraintSetup [
     {transformType=Rotation, targetBone {boneName=spine_05}}
     {transformType=Rotation, targetBone {boneName=clavicle_r}}
     {transformType=Rotation, targetBone {boneName=neck_01}}
    ]
    constraintWeights [1,1,1]
   }
   n2 = "Constraint - Collar_bk_MSTR (RR)"!(ComponentPose=@n1, Alpha=1, ConstraintWeights_0=0.5, ConstraintWeights_1:float=0.5) -> (Pose) {
    alphaBoolBlend {blendOption=Linear}
    boneToModify {boneName=Collar_bk_MSTR}
    constraintSetup [{transformType=Rotation, targetBone {boneName=spine_05}}, {transformType=Rotation, targetBone {boneName=neck_01}}]
    constraintWeights [1,1]
   }
   n3 = "Constraint - Collar_bk_l_MSTR (RRR)"!(ComponentPose=@n2, Alpha=1, ConstraintWeights_0=0.25, ConstraintWeights_1:float=0.25, ConstraintWeights_2:float=0.5) -> (Pose) {
    alphaBoolBlend {blendOption=Linear}
    boneToModify {boneName=Collar_bk_l_MSTR}
    constraintSetup [
     {transformType=Rotation, targetBone {boneName=spine_05}}
     {transformType=Rotation, targetBone {boneName=clavicle_l}}
     {transformType=Rotation, targetBone {boneName=neck_01}}
    ]
    constraintWeights [1,1,1]
   }
   n4 = "Constraint - Collar_frt_r_MSTR (PP)"!(ComponentPose=@n3, Alpha=1, ConstraintWeights_0=0.5, ConstraintWeights_1:float=0.5) -> (Pose) {
    alphaBoolBlend {blendOption=Linear}
    boneToModify {boneName=Collar_frt_r_MSTR}
    constraintSetup [{targetBone {boneName=neck_01}}, {targetBone {boneName=clavicle_r}}]
    constraintWeights [1,1]
   }
   n5 = "Constraint - Collar_frt_l_MSTR (PP)"!(ComponentPose=@n4, Alpha=1, ConstraintWeights_0=0.5, ConstraintWeights_1:float=0.5) -> (Pose) {
    alphaBoolBlend {blendOption=Linear}
    boneToModify {boneName=Collar_frt_l_MSTR}
    constraintSetup [{targetBone {boneName=neck_01}}, {targetBone {boneName=clavicle_l}}]
    constraintWeights [1,1]
   }
   n6 = "Constraint - Skirt_bk_l_MSTR (PP)"!(ComponentPose=@n5, Alpha=1, ConstraintWeights_0=0.8, ConstraintWeights_1:float=0.2) -> (Pose) {
    alphaBoolBlend {blendOption=Linear}
    boneToModify {boneName=Skirt_bk_l_MSTR}
    constraintSetup [{targetBone {boneName=pelvis}}, {targetBone {boneName=thigh_l}}]
    constraintWeights [1,1]
   }
   n7 = "Constraint - Skirt_bk_r_MSTR (PP)"!(ComponentPose=@n6, Alpha=1, ConstraintWeights_0=0.8, ConstraintWeights_1:float=0.2) -> (Pose) {
    alphaBoolBlend {blendOption=Linear}
    boneToModify {boneName=Skirt_bk_r_MSTR}
    constraintSetup [{targetBone {boneName=pelvis}}, {targetBone {boneName=thigh_r}}]
    constraintWeights [1,1]
   }
   n8 = "Transform (Modify) Bone - Bone: pelvis_adjust_jnt"(ComponentPose=@n7, Alpha=Get Is Crouching Alpha() -> (Is Crouching Alpha:double), Translation="1.5,0,0", Rotation="0,0,0", Scale="1,1,1") {
    rotationMode=BMM_Additive
    rotationSpace=BCS_BoneSpace
    translationMode=BMM_Additive
    translationSpace=BCS_BoneSpace
    alphaBoolBlend {blendOption=Linear}
    boneToModify {boneName=pelvis_adjust_jnt}
    translation {x=1.5}
   }
   n9 = Get Is Parachute Opened Alpha() -> (Is Parachute Opened Alpha:double)
   n10 = "Transform (Modify) Bone - Bone: Skirt_bk_l_MSTR"(ComponentPose=@n8, Alpha=@n9, Translation="0,1,1", Rotation="5,5,0", Scale="1,1,1") {
    rotationMode=BMM_Additive
    rotationSpace=BCS_BoneSpace
    translationMode=BMM_Additive
    translationSpace=BCS_BoneSpace
    alphaBoolBlend {blendOption=Linear}
    boneToModify {boneName=Skirt_bk_l_MSTR}
    rotation {pitch=5, yaw=5}
    translation {y=1, z=1}
   }
   n11 = "Transform (Modify) Bone - Bone: Skirt_bk_r_MSTR"(ComponentPose=@n10, Alpha=@n9, Translation="0,-1,1", Rotation="5,-5,0", Scale="1,1,1") {
    rotationMode=BMM_Additive
    rotationSpace=BCS_BoneSpace
    translationMode=BMM_Additive
    translationSpace=BCS_BoneSpace
    alphaBoolBlend {blendOption=Linear}
    boneToModify {boneName=Skirt_bk_r_MSTR}
    rotation {pitch=5, yaw=-5}
    translation {y=-1, z=1}
   }
   n12 = Control Rig!(
    Source=Component To Local(ComponentPose=@n11)
    Alpha
    Is Pistol Alpha:double=Get Is Pistol Alpha() -> (Is Pistol Alpha:double)
    Is Not Playing Emote:double=Get Is not Playing Emote Alpha() -> (Is not Playing Emote Alpha:double)
   ) -> (Pose) {
    alphaBoolBlend {blendOption=Linear}
    controlRigAssetReference {blueprintRigClass=&/F_MED_SkeleProbe_CtrlRig.F_MED_SkeleProbe_CtrlRig_C}
   }
   Output Pose(Result=@n12) {layerGroup=None}
  ]}
 EventGraph {nodes [
   Event BlueprintUpdateAnimation()
   > n1 = Sequence!(execute) -> (then_0, then_1, then_2:Exec, then_3:Exec, then_4:Exec)
   > n2 = Sequence()
   > n3 = Cast To FortPlayerAnimInstance(Object=GetAnimInstance(self=GetOwnerSkeletalMeshComponent()))
   > Set isPistol(isPistol=false)
   > Set isRifle(isRifle:bool=false) -> (Output_Get:bool)
   > n4 = Switch on EFortWeaponCoreAnimation!(execute, Selection:EFortWeaponCoreAnimation Enum=Get WeaponCoreAnim(self=@n3)) -> (
    Melee:Exec
    Pistol:Exec
    Shotgun:Exec
    PaperBlueprint:Exec
    Rifle:Exec
    MeleeOneHand:Exec
    MachinePistol:Exec
    RocketLauncher:Exec
    GrenadeLauncher:Exec
    GoingCommando:Exec
    AssaultRifle:Exec
    TacticalShotgun:Exec
    SniperRifle:Exec
    TrapPlacement:Exec
    ShoulderLauncher:Exec
    AbilityDecoTool:Exec
    Crossbow:Exec
    C4:Exec
    RemoteControl:Exec
    DualWield:Exec
    AR_BullPup:Exec
    AR_ForwardGrip:Exec
    MedPackPaddles:Exec
    SMG_P90:Exec
    AR_DrumGun:Exec
    Consumable_Small:Exec
    Consumable_Large:Exec
    Balloon:Exec
    MountedTurret:Exec
    CreativeTool:Exec
    ExplosiveBow:Exec
    AshtonIndigo:Exec
    AshtonChicago:Exec
    MeleeDualWield:Exec
    Unarmed:Exec
   )
   Set isRifle(execute=[@n4.AssaultRifle, @n4.Rifle, @n4.RocketLauncher, @n4.Shotgun, @n4.SniperRifle, @n4.TacticalShotgun], isRifle=true)
   Set isPistol(execute=@n4.Pistol, isPistol=true)
   Set Is Rifle Alpha(execute=@n2.then_1, Is Rifle Alpha:double="To Float (Boolean)"(InBool=Get isRifle() -> (isRifle:bool))) -> (Output_Get:double)
   Set Is Pistol Alpha(execute=@n2.then_2, Is Pistol Alpha:double="To Float (Boolean)"(InBool=Get isPistol() -> (isPistol:bool))) -> (Output_Get:double)
   n5 = Get bIsPlayingEmote()
   Set Is not Playing Emote Alpha(execute=@n1.then_1, Is not Playing Emote Alpha:double="To Float (Boolean)"(InBool=NOT Boolean(A=@n5))) -> (Output_Get:double)
   Set Is Crouching Alpha(execute=@n1.then_2, Is Crouching Alpha:double="To Float (Boolean)"(InBool=Get bIsCrouching())) -> (Output_Get:double)
   Set Is Parachute Opened Alpha(execute=@n1.then_3, Is Parachute Opened Alpha:double="To Float (Boolean)"(InBool=Get bIsParachuteOpened())) -> (Output_Get:double)
  ]}
]
variables [
 Is Crouching Alpha {type=number}
 Is not Playing Emote Alpha {type=number}
 Is Parachute Opened Alpha {type=number}
 Is Pistol Alpha {type=number}
 Is Rifle Alpha {type=number}
 isPistol {type=boolean}
 isRifle {type=boolean}
]
